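; Gameplay ability managing the lifetime of the elimination cloak boon (Nightmare Cloak).
; Activation adds the cloak cue and, when a scalable-float hotfix is on, ends the cloak on a damage-received event.
; A delay caps duration; removal strips the granting effect by tag, ending the ability, whose end handler also stops the damage task.
; A helper checks the avatar's Kor leap cloaked tag but is uncalled here; three other events are empty.
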(event Custom|RemoveCloak
  ;[1 1 1] Remove Cloak
  (Ability|RemoveGameplayEffectFromOwnerWithGrantedTags (Variables|Tags|GetTag_NightmareCloak_Active)) ; Removes granting effect, ending ability
)

(event Custom|StartGameplayEventManager)

(event Custom|HideFromAI)

(event Ability|EventActivateAbility
  ;[1 1 1] On Activate Ability
  (|CallParentFunction)
  (Ability|AddGameplayCueToOwner (Variables|Tags|GetTag_GC_NightmareCloak))
  (Utilities|FlowControl|Sequence
    (:then_0
      ;[1 1 1 0.5] Adding Bool Hotfix for Damage on Cancel
      (if (> (ScalableFloat|GetValueatLevel (Variables|Default|GetShouldCancelOnDamage)) 0)
        (bind _asynctaskproxy (Ability|Tasks|WaitGameplayEvent "(TagName=\"Event.Damage.Received\")")
          (:then
            (Variables|Default|SetAsync%20Task%20Damage%20Received _asynctaskproxy))
          (:EventReceived
            (CallFunction|RemoveCloak)))))
    (:then_1
      (Utilities|FlowControl|Delay (ScalableFloat|GetValueatLevel (Variables|Default|GetMaxCloakDuration) (Variables|Default|GetMaxCloakDurationLevel)))
      (CallFunction|RemoveCloak))))

(event Ability|EventOnEndAbility
  (params (bWasCancelled :type Boolean))
  ;[1 1 1] On End Ability
  (|CallParentFunction bWasCancelled)
  (CallFunction|RemoveCloak)
  (bind _async_task_damage_received (Variables|Default|GetAsync%20Task%20Damage%20Received)
    (:then
      (GameplayTasks|EndTask _async_task_damage_received))
    (:else)))

(event Custom|CheckWaterDepth)

(fn CheckKorLeapStatus
  (return (GameplayTags|HasMatchingGameplayTag (Utilities|Casting|ToGameplayTagAssetInterface (Ability|GetAvatarActorfromActorInfo)) "(TagName=\"Event.CherrySmoke.Medallions.Leap.Cloaked\")")))
