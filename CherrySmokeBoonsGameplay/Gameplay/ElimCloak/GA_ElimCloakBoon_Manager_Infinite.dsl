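; Gameplay ability managing the Hide and Seek cloak boon, which cloaks a player who stays still or crouched.
; A repeating timer counts clean checks up to a threshold, then applies the cloak cue and effect.
; Movement, ADS, sliding, disallowed tags, damage, firing, melee or building reset it through polling, event tasks and tag bindings, with falling and DBNO exceptions.
; Losing the rift anomaly participant tag terminates the cloak and ability.
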(event Custom|RemoveCloak
  ;[1 1 1] Remove Cloak
  (Ability|RemoveGameplayCueFromOwner (Variables|Tags|GetTag_GC_HSCloak)))

(event Custom|RemoveGameplayEffect
  ;[1 1 1] Remove Cloak
  (Ability|RemoveGameplayEffectFromOwnerWithGrantedTags (Variables|Tags|GetTag_HSCloak_Tags)) ; Removes granting effect, ending ability
  (Ability|RemoveGameplayEffectFromOwnerWithGrantedTags (Variables|Tags|GetTag_HSCloak_Active)) ; Removes granting effect, ending ability
)

(event Custom|StartGameplayEventManager
  (bind _returnvalue (Ability|GetAbilitySystemComponentfromActorInfo))
  ;[1 1 1 0.5] Adding Bool Hotfix for Damage on Cancel
  (bind _asynctaskproxy (Ability|Tasks|WaitGameplayEvent (Variables|Tags|GetTag_FallingDamage))
    (:then
      (Variables|Default|SetAsyncTaskFallingDamage _asynctaskproxy)
      (bind (_asynctaskproxy_2 _payload) (Ability|Tasks|WaitGameplayEvent "(TagName=\"Event.Damage.Overshielded\")")
        (:then
          (Variables|Default|SetAsyncTaskDamageDealtOvershielded _asynctaskproxy_2)
          (bind (_asynctaskproxy_3 _payload_2) (Ability|Tasks|WaitGameplayEvent "(TagName=\"Event.Damage.Shielded\")")
            (:then
              (Variables|Default|SetAsyncTaskDamageDealtShielded _asynctaskproxy_3)
              (bind (_asynctaskproxy_4 _payload_3) (Ability|Tasks|WaitGameplayEvent "(TagName=\"Event.Damage.Dealt\")")
                (:then
                  (Variables|TimerAndTaskHandles|SetAsyncTaskDamageDealt _asynctaskproxy_4)
                  (bind (_asynctaskproxy_5 _payload_4) (Ability|Tasks|WaitGameplayEvent "(TagName=\"Event.Damage.Received\")")
                    (:then
                      (Variables|TimerAndTaskHandles|SetAsync%20Task%20Damage%20Received _asynctaskproxy_5)
                      (bind _asynctaskproxy_6 (Ability|Tasks|WaitGameplayEvent "(TagName=\"Event.Weapons.Fired\")")
                        (:then
                          (Variables|TimerAndTaskHandles|SetAsync%20Task%20Weapons%20Fired _asynctaskproxy_6)
                          (bind _asynctaskproxy_7 (Ability|Tasks|WaitGameplayEvent "(TagName=\"Event.Weapons.Fire\")")
                            (:then
                              (Variables|TimerAndTaskHandles|SetAsync%20Task%20Weapons%20Fire _asynctaskproxy_7)
                              (bind _asynctaskproxy_8 (Ability|Tasks|WaitGameplayEvent "(TagName=\"Event.Weapon.Fired\")")
                                (:then
                                  (Variables|TimerAndTaskHandles|SetAsync%20Task%20Weapon%20Fired _asynctaskproxy_8)
                                  (bind _returnvalue_2 (EventProcessorSubsystem|StartListeningToFilteredEvent 0 (Ability|GetAvatarActorfromActorInfo) (Variables|Default|GetBuildVerb) false (delegate Custom|BuildVerbEvent)))
                                  (Variables|Default|SetBuildVerbHandle _returnvalue_2)
                                  (bind _returnvalue_3 (GameplayTags|BindEventWrappertoGameplayTagChanged _returnvalue (Variables|Tags|GetADSGameplayTag) (delegate Custom|OnADSTagChanged)))
                                  (Utilities|Array|AddUnique (Variables|Default|GetGameplayTagChangedBindings) _returnvalue_3)
                                  (bind _returnvalue_4 (GameplayTags|BindEventWrappertoGameplayTagChanged _returnvalue "(TagName=\"Gameplay.Action.Player.Crouch\")" (delegate Custom|OnCrouchTagChanged)))
                                  (Utilities|Array|AddUnique (Variables|Default|GetGameplayTagChangedBindings) _returnvalue_4)
                                  (bind _returnvalue_5 (GameplayTags|BindEventWrappertoGameplayTagChanged _returnvalue "(TagName=\"MovementMode.Sliding\")" (delegate Custom|SlideWaitEvent)))
                                  (Utilities|Array|AddUnique (Variables|Default|GetGameplayTagChangedBindings) _returnvalue_5)
                                  (bind _returnvalue_6 (GameplayTags|BindEventWrappertoGameplayTagChanged _returnvalue "(TagName=\"Gameplay.Action.Attacking.Melee\")" (delegate Custom|OnMeleeTagChanged)))
                                  (Utilities|Array|AddUnique (Variables|Default|GetGameplayTagChangedBindings) _returnvalue_6))
                                (:EventReceived
                                  _)))
                            (:EventReceived
                              _))
                          (CallFunction|ResetCloakStatus)
                          (Development|PrintString "Hide and Seek Cloak Should Reset!"))
                        (:EventReceived
                          _)))
                    (:EventReceived
                      (bind _allowed (CallFunction|CheckifAllowDBNOCloak _payload_4))
                      (if _allowed
                        _))))
                (:EventReceived
                  (bind _allowed_2 (CallFunction|CheckifAllowDBNOCloak _payload_3))
                  (if _allowed_2
                    _))))
            (:EventReceived
              (bind _allowed_3 (CallFunction|CheckifAllowDBNOCloak _payload_2))
              (if _allowed_3
                _))))
        (:EventReceived
          (bind _allowed_4 (CallFunction|CheckifAllowDBNOCloak _payload))
          (if _allowed_4
            _))))
    (:EventReceived
      _)))

(event Custom|OnMeleeTagChanged
  (params (Tag :type Gameplay%20Tag%20Structure%20%28by%20ref%29) (TagCount :type Integer))
  ;[1 1 1 0.5] Adding Bool Hotfix for Damage on Cancel
  (if (> TagCount 0)
    (CallFunction|ResetCloakStatus)
    (Development|PrintString "Hide and Seek Cloak Should Reset!")))

(event Custom|OnADSTagChanged
  (params (Tag :type Gameplay%20Tag%20Structure%20%28by%20ref%29) (TagCount :type Integer))
  ;[1 1 1 0.5] Adding Bool Hotfix for Damage on Cancel
  (if (> TagCount 0)
    (Utilities|FlowControl|Sequence
      (:then_0
        (CallFunction|ResetCloakStatus)
        (Development|PrintString "Hide and Seek Cloak Should Reset!"))
      (:then_1
        (Variables|Default|SetIsADS true)))
    (else
      (Variables|Default|SetIsADS false))))

(event Custom|BuildVerbEvent
  ;[1 1 1 0.5] Adding Bool Hotfix for Damage on Cancel
  (CallFunction|ResetCloakStatus)
  (Development|PrintString "Hide and Seek Cloak Should Reset!"))

(event Custom|OnCrouchTagChanged
  (params (Tag :type Gameplay%20Tag%20Structure%20%28by%20ref%29) (TagCount :type Integer))
  ;[1 1 1 0.5] Adding Bool Hotfix for Damage on Cancel
  (if (> TagCount 0)
    (bind _confirmed (CallFunction|VerifyCrouchStatus))
    (Variables|Default|SetIsCrouching _confirmed)
    (else
      (Utilities|FlowControl|RetriggerableDelay 0.5)
      (bind _confirmed_2 (CallFunction|VerifyCrouchStatus))
      (Variables|Default|SetIsCrouching _confirmed_2))))

(event Custom|SlideWaitEvent
  (params (Tag :type Gameplay%20Tag%20Structure%20%28by%20ref%29) (TagCount :type Integer))
  ;[1 1 1 0.5] Adding Bool Hotfix for Damage on Cancel
  (if (> TagCount 0)
    (Variables|Default|SetIsSliding true)
    (CallFunction|ResetCloakStatus)
    (Development|PrintString "Hide and Seek Cloak Should Reset!")
    (else
      (Utilities|FlowControl|RetriggerableDelay 0.5)
      (bind _confirmed (CallFunction|VerifySlidingStatus))
      (Variables|Default|SetIsSliding _confirmed))))

(event Ability|EventActivateAbility
  (bind _returnvalue (Ability|GetAvatarActorfromActorInfo))
  ;[1 1 1] On Activate Ability
  (|CallParentFunction)
  (CallFunction|HidefromAI)
  (CallFunction|StartGameplayEventManager)
  (Development|PrintString "HIDE AND SEEK GA ACTIVATED" true true "(R=0.412737,G=0.000326,B=1.000000,A=1.000000)" 5.0)
  ;[1 1 1] Primary Hide and Seek Logic
  (if (Utilities|Time|IsValidTimerHandle (Variables|Default|GetVelocityCheckTimer))
    (Utilities|Time|ClearandInvalidateTimerbyHandle (Variables|Default|GetVelocityCheckTimer))
    (else
      _))
  (Utilities|IsValid _returnvalue
    (:Is%20Valid
      (bind _output_get (Variables|Default|SetPlayerPawn _returnvalue))
      (Variables|Default|SetLastLocation "0, 0, 0")
      (bind _returnvalue_2 (Utilities|Time|SetTimerbyEvent (delegate Custom|VelocityCheck) (ScalableFloat|GetValueatLevel (Variables|Default|GetActivationTimerValue)) true true))
      (Variables|Default|SetVelocityCheckTimer _returnvalue_2)
      (Development|PrintString "HideAndSeek Cloak STARTING VC TIMER!"))
    (:Is%20Not%20Valid
      (Development|PrintString "HideAndSeek Cloak Avatar Actor Valid Check FAILED!"))))

(event Custom|RestartCloakSequence
  (bind _returnvalue (Ability|GetAvatarActorfromActorInfo))
  ;[1 1 1] On Activate Ability
  (|CallParentFunction)
  (CallFunction|HidefromAI)
  (CallFunction|StartGameplayEventManager)
  (Development|PrintString "HIDE AND SEEK GA ACTIVATED" true true "(R=0.412737,G=0.000326,B=1.000000,A=1.000000)" 5.0)
  ;[1 1 1] Primary Hide and Seek Logic
  (if (Utilities|Time|IsValidTimerHandle (Variables|Default|GetVelocityCheckTimer))
    (Utilities|Time|ClearandInvalidateTimerbyHandle (Variables|Default|GetVelocityCheckTimer))
    (else
      _))
  (Utilities|IsValid _returnvalue
    (:Is%20Valid
      (bind _output_get (Variables|Default|SetPlayerPawn _returnvalue))
      (Variables|Default|SetLastLocation "0, 0, 0")
      (bind _returnvalue_2 (Utilities|Time|SetTimerbyEvent (delegate Custom|VelocityCheck) (ScalableFloat|GetValueatLevel (Variables|Default|GetActivationTimerValue)) true true))
      (Variables|Default|SetVelocityCheckTimer _returnvalue_2)
      (Development|PrintString "HideAndSeek Cloak STARTING VC TIMER!"))
    (:Is%20Not%20Valid
      (Development|PrintString "HideAndSeek Cloak Avatar Actor Valid Check FAILED!"))))

(event Custom|VelocityCheck
  ;[1 1 1] Primary Hide and Seek Logic
  (Utilities|FlowControl|Sequence
    (:then_0
      ;;[1 1 1 0.5] Ensures that players already crouching or sliding as this GA is granted properly apply cloak rules on first application
      (Utilities|FlowControl|DoOnce false
        (:Completed
          (bind _confirmed (CallFunction|VerifyCrouchStatus))
          (Variables|Default|SetIsCrouching _confirmed)
          (bind _confirmed_2 (CallFunction|VerifySlidingStatus))
          (Variables|Default|SetIsSliding _confirmed_2))))
    (:then_1
      (CallFunction|CheckIsPlayerFalling)
      (bind _verified (CallFunction|VerifyActiveRiftAnomaly))
      (if _verified
        (Utilities|IsValid (Variables|Default|GetPlayerPawn)
          (:Is%20Valid
            (CallFunction|CheckforIllegalTags)
            (CallFunction|UpdateCrouchedLocation)
            (if (Math|Boolean|ORBoolean (Variables|Default|GetIsADS) (Variables|Default|GetArePlayerTagsIllegal) (and (> (Math|Vector|VectorLength (- (Transformation|GetActorLocation (Variables|Default|GetPlayerPawn)) (Variables|Default|GetLastLocation))) (Variables|Default|GetDecloakSensitivityThreshold)) (not (Variables|Default|GetIsCrouching))) (Variables|Default|GetIsSliding) (Variables|Default|GetIsFiringA51Gun) (Variables|Default|GetIsA51CloakActive))
              (if (not (Variables|Default|GetIsPlayerFalling))
                (CallFunction|ResetCloakStatus)
                (Ability|RemoveGameplayEffectFromOwnerWithGrantedTags (Variables|Tags|GetTag_HSCloak_Tags)) ; Removes granting effect, ending ability
              )
              (elif (not (Variables|Default|GetCloakEngaged))
                (Math|Integer|IncrementInt (Variables|Default|GetCloakWaitCountThreshold)
                  (:%20%20
                    (if (>= (Variables|Default|GetCloakWaitCountThreshold) (ScalableFloat|GetValueatLevel (Variables|Default|GetActivationTimerThresholdValue)))
                      (Variables|Default|SetCloakEngaged true)
                      (Ability|AddGameplayCueToOwner (Variables|Tags|GetTag_GC_HSCloak))
                      (Fortnite|ApplyGameplayEffectsToActor (Variables|Default|GetPlayerPawn) (Utilities|Array|MakeArray (Variables|Default|GetHSCloakAppliedGE)) (Transformation|GetActorLocation (Variables|Default|GetPlayerPawn)) false 0 (Ability|GetFortAbilitySystemComponent (Variables|Default|GetPlayerPawn)) (Ability|GetContextfromOwner))))))))
          (:Is%20Not%20Valid
            (Development|PrintString "HideAndSeek Cloak Avatar Actor Valid Check FAILED!")))
        (else
          (CallFunction|TerminateCloak))))))

(event Custom|ForceApplyCloak
  ;[1 1 1] Primary Hide and Seek Logic
  (Ability|AddGameplayCueToOwner (Variables|Tags|GetTag_GC_HSCloak))
  (Fortnite|ApplyGameplayEffectsToActor (Variables|Default|GetPlayerPawn) (Utilities|Array|MakeArray (Variables|Default|GetHSCloakAppliedGE)) (Transformation|GetActorLocation (Variables|Default|GetPlayerPawn)) false 0 (Ability|GetFortAbilitySystemComponent (Variables|Default|GetPlayerPawn)) (Ability|GetContextfromOwner)))

(event Ability|EventOnEndAbility
  (params (bWasCancelled :type Boolean))
  ;[1 1 1] On End Ability
  (|CallParentFunction bWasCancelled)
  (Utilities|Time|ClearandInvalidateTimerbyHandle (Variables|TimerandTaskHandles|GetHandle_HideFromAITimer))
  (Utilities|IsValid (Ability|GetAvatarActorfromActorInfo)
    (:Is%20Valid
      (Ability|RemoveGameplayEffectFromOwnerWithHandle (Variables|Default|GetGEHandle_HideFromAI))
      (CallFunction|ResetCloakStatus)
      (bind _async_task_damage_received (Variables|TimerAndTaskHandles|GetAsync%20Task%20Damage%20Received)
        (:then
          (GameplayTasks|EndTask _async_task_damage_received)
          (GameplayTasks|EndTask (Variables|TimerAndTaskHandles|GetAsync%20Task%20Weapons%20Fired))
          (for _array_element (Variables|Default|GetGameplayTagChangedBindings)
            (GameplayTags|UnbindAllGameplayTagChangedEventWrappersforHandle _array_element)))
        (:else)))))

(event Custom|ResetCloakStatus
  ;[1 1 1] Primary Hide and Seek Logic
  (Variables|Default|SetCloakWaitCountThreshold 0)
  (Variables|Default|SetCloakEngaged false)
  (Variables|Default|SetLastLocation (Transformation|GetActorLocation (Variables|Default|GetPlayerPawn)))
  (CallFunction|RemoveCloak)
  (if (Variables|Default|GetShouldEndCloakVCTimer)
    (if (Utilities|Time|IsValidTimerHandle (Variables|Default|GetVelocityCheckTimer))
      (Utilities|Time|ClearandInvalidateTimerbyHandle (Variables|Default|GetVelocityCheckTimer)))))

(event Custom|TerminateCloak
  ;[1 1 1] Primary Hide and Seek Logic
  (CallFunction|RemoveCloak)
  (Utilities|Time|ClearandInvalidateTimerbyHandle (Variables|Default|GetVelocityCheckTimer))
  (CallFunction|RemoveGameplayEffect)
  (Ability|EndAbility))

(event Custom|HideFromAI)

(fn CheckForIllegalTags
  (bind _returnvalue (GameplayTags|HasAnyMatchingGameplayTags(Message) (Variables|Default|GetPlayerPawn) (GameplayTags|MakeLiteralGameplayTagContainer "(GameplayTags=((TagName=\"MovementMode.Driving\"),(TagName=\"MovementMode.Passenger\"),(TagName=\"MovementMode.Parachuting\"),(TagName=\"MovementMode.Ballooning\"),(TagName=\"MovementMode.DBNOCarried\"),(TagName=\"MovementMode.Flying\"),(TagName=\"MovementMode.Skydiving\"),(TagName=\"MovementMode.SurfaceSwimming\"),(TagName=\"MovementMode.Custom.HangTime\"),(TagName=\"MovementMode.Sliding\"),(TagName=\"MovementMode.TacSprint\"),(TagName=\"Granted.Athena.Item.MedConsumable.Channeling\"),(TagName=\"Gameplay.Action.Player.CarryingPlayer\"),(TagName=\"Gameplay.Action.Player.InteractSearch.DBNO\"),(TagName=\"Gameplay.Action.Player.InteractSearch.Reboot\"),(TagName=\"Granted.Mud.CoveredInMud\"),(TagName=\"Granted.Athena.BuiltInEmote\"),(TagName=\"Gameplay.Status.DBNOBeingRevived\"),(TagName=\"Gameplay.Action.Player.Jam\"),(TagName=\"Gameplay.Action.Player.Emote\"),(TagName=\"Pawn.Athena.NonParticipant\"),(TagName=\"Pawn.Athena.NPC\"),(TagName=\"Granted.UmbramoltKipper.AirHold.LowGrav\"),(TagName=\"Granted.UmbramoltKipper.Dash.Active\"),(TagName=\"Granted.UmbramoltKipper.AttackBlocked\"),(TagName=\"Granted.UmbramoltKipper.Charging\"),(TagName=\"Granted.UmbramoltKipper.HoverMMEActive\"),(TagName=\"Granted.UmbramoltKipper.Diving\"),(TagName=\"Gameplay.Status.Rift.Teleported\"),(TagName=\"RiftAnomalies.HideAndSeek.MarkMe\"),(TagName=\"MovementMode.Slippery\"),(TagName=\"Gameplay.Action.Attacking.Melee\"),(TagName=\"Gameplay.Action.Player.ThrowConsumable\"),(TagName=\"Gameplay.ProtoCannon.Charging\")))")))
  (Variables|Default|SetArePlayerTagsIllegal _returnvalue)
  ; Area 51 Gun Special Check
  (bind _returnvalue_2 (GameplayTags|HasMatchingGameplayTag(Message) (Variables|Default|GetPlayerPawn) (Variables|Default|GetA51CueTag)))
  (Variables|Default|SetIsFiringA51Gun _returnvalue_2)
  ; Area 51 Goo Special Check
  (bind _returnvalue_3 (GameplayTags|HasAnyMatchingGameplayTags(Message) (Variables|Default|GetPlayerPawn) (GameplayTags|MakeLiteralGameplayTagContainer "(GameplayTags=((TagName=\"SecretFacility.Invisibility.Enabled\"),(TagName=\"Item.Consumable.StealthSplash.Cloak.Active\")))")))
  (Variables|Default|SetIsA51CloakActive _returnvalue_3)
  (return))

(fn CheckIfAllowDBNOCloak
  (params (Payload :type Gameplay%20Event%20Data%20Structure))
  (bind (_eventtag _instigator _target _optionalobject) (Utilities|Struct|BreakGameplayEventData Payload))
  (bind _returnvalue (GameplayTags|HasMatchingGameplayTag(Message) _optionalobject (Variables|Default|GetTag_OutsideSafeZoneDamage)))
  (if (not _returnvalue)
    (bind _returnvalue_2 (GameplayTags|HasMatchingGameplayTag(Message) (Variables|Default|GetPlayerPawn) (Variables|Tags|GetTag_GameplayStatus_DBNO)))
    (if _returnvalue_2
      (if (> (ScalableFloat|GetValueatLevel (Variables|Default|GetAllowDBNOCloakValue)) 0)
        (return false)
        (else
          (return true)))
      (else
        (return true)))
    (else
      (return true))))

(fn CheckIsPlayerFalling
  (bind _returnvalue (GameplayTags|HasMatchingGameplayTag(Message) (Variables|Default|GetPlayerPawn) "(TagName=\"MovementMode.Falling\")"))
  (Variables|Default|SetIsPlayerFalling _returnvalue)
  (return))

(fn CheckKorLeapStatus
  (return (GameplayTags|HasMatchingGameplayTag (Utilities|Casting|ToGameplayTagAssetInterface (Ability|GetAvatarActorfromActorInfo)) "(TagName=\"Event.CherrySmoke.Medallions.Leap.Cloaked\")")))

(fn UpdateCrouchedLocation
  (if (Variables|Default|GetIsCrouching)
    (Variables|Default|SetLastLocation (Transformation|GetActorLocation (Variables|Default|GetPlayerPawn)))
    (return)
    (else
      (return))))

(fn VerifyActiveRiftAnomaly
  (bind _returnvalue (GameplayTags|HasMatchingGameplayTag(Message) (Variables|Default|GetPlayerPawn) "(TagName=\"RiftAnomalies.HideAndSeek.Participant\")"))
  (return _returnvalue))

(fn VerifyCrouchStatus
  (bind _returnvalue (GameplayTags|HasMatchingGameplayTag(Message) (Variables|Default|GetPlayerPawn) (Variables|Tags|GetCrouchGameplayTag)))
  (if _returnvalue
    (bind _returnvalue_2 (GameplayTags|HasMatchingGameplayTag(Message) (Variables|Default|GetPlayerPawn) (Variables|Tags|GetTag_GameplayStatus_DBNO)))
    (if _returnvalue_2
      (return false)
      (else
        (return true)))
    (else
      (return false))))

(fn VerifySlidingStatus
  (bind _returnvalue (GameplayTags|HasMatchingGameplayTag(Message) (Variables|Default|GetPlayerPawn) (Variables|Tags|GetSlidingTag)))
  (if _returnvalue
    (return true)
    (else
      (return false))))
