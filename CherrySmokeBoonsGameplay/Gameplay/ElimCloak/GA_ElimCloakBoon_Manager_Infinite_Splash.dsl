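; Gameplay ability for the CherrySmoke elimination cloak boon that shows the Nightmare Cloak gameplay cue on its owner for up to 20 seconds.
; Staggered wait tasks listen for damage and weapon fire events to strip the cue early.
; Ending the ability clears timers, the hide-from-AI effect handle, the cue and some tasks.
; A stand-still re-cloak check exists but is not called here, and several helper events are empty stubs.
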(event Custom|RemoveCloak
  (Ability|RemoveGameplayCueFromOwner (Variables|Tags|GetTag_GC_NightmareCloak)))

(event Custom|StartGameplayEventManager)

(event Custom|HideFromAI)

(event Ability|EventActivateAbility
  ;[1 1 1] On Activate Ability
  (|CallParentFunction)
  (Utilities|FlowControl|Sequence
    (:then_0
      ;[1 1 1 0.5] Adding Bool Hotfix for Damage on Cancel
      (Utilities|FlowControl|Delay 0.5)
      (bind _asynctaskproxy (Ability|Tasks|WaitGameplayEvent "(TagName=\"Event.Damage.Received\")")
        (:then
          (Variables|Default|SetAsync%20Task%20Damage%20Received _asynctaskproxy)
          (Utilities|FlowControl|Delay 0.5)
          (bind _asynctaskproxy_2 (Ability|Tasks|WaitGameplayEvent "(TagName=\"Event.Weapons.Fired\")")
            (:then
              (Variables|TimerAndTaskHandles|SetAsync%20Task%20Weapons%20Fired _asynctaskproxy_2)
              (bind _asynctaskproxy_3 (Ability|Tasks|WaitGameplayEvent "(TagName=\"Event.Weapons.Fire\")")
                (:then
                  (Variables|Default|SetAsync%20Task%20Weapons%20Fire _asynctaskproxy_3))
                (:EventReceived
                  _)))
            (:EventReceived
              _))
          ;[1 1 1] Gameplay Event Manager
          (CallFunction|RemoveCloak))
        (:EventReceived
          (CallFunction|RemoveCloak))))
    (:then_1
      (Ability|AddGameplayCueToOwner (Variables|Tags|GetTag_GC_NightmareCloak)))
    (:then_2
      (Utilities|FlowControl|Delay 2e+01)
      (CallFunction|RemoveCloak)
      (Ability|EndAbility))))

(event Ability|EventOnEndAbility
  (params (bWasCancelled :type Boolean))
  ;[1 1 1] On End Ability
  (|CallParentFunction bWasCancelled)
  (Utilities|Time|ClearandInvalidateTimerbyHandle (Variables|TimerandTaskHandles|GetHandle_HideFromAITimer))
  (Utilities|Time|ClearandInvalidateTimerbyHandle (Variables|Default|GetVelocityCheckTimer))
  (Utilities|IsValid (Ability|GetAvatarActorfromActorInfo)
    (:Is%20Valid
      (Ability|RemoveGameplayEffectFromOwnerWithHandle (Variables|Default|GetGEHandle_HideFromAI))
      (CallFunction|RemoveCloak)
      (bind _async_task_damage_received (Variables|Default|GetAsync%20Task%20Damage%20Received)
        (:then
          (GameplayTasks|EndTask _async_task_damage_received)
          (GameplayTasks|EndTask (Variables|TimerAndTaskHandles|GetAsync%20Task%20InBushProp)))
        (:else)))))

(event Custom|VelocityCheck
  ;[1 1 1] Is Player Standing Sill?
  (Utilities|IsValid (Variables|Default|GetPlayerPawn)
    (:Is%20Valid
      (if (> (Math|Vector|VectorLength (- (Transformation|GetActorLocation (Variables|Default|GetPlayerPawn)) (Variables|Default|GetLastLocation))) (Variables|Default|GetDecloakSensitivityThreshold))
        (Variables|Default|SetCloakWaitCountThreshold 0)
        (Variables|Default|SetCloakEngaged false)
        (Variables|Default|SetLastLocation (Transformation|GetActorLocation (Variables|Default|GetPlayerPawn)))
        (CallFunction|RemoveCloak)
        (elif (not (Variables|Default|GetCloakEngaged))
          (Math|Integer|IncrementInt (Variables|Default|GetCloakWaitCountThreshold)
            (:%20%20
              (if (>= (Variables|Default|GetCloakWaitCountThreshold) 2)
                (Variables|Default|SetCloakEngaged true)
                (Ability|AddGameplayCueToOwner (Variables|Tags|GetTag_GC_NightmareCloak))))))))
    (:Is%20Not%20Valid
      (Development|PrintString "HideAndSeek Cloak Avatar Actor Valid Check FAILED!"))))

(event Custom|CheckWaterDepth)

(fn CheckKorLeapStatus
  (return (GameplayTags|HasMatchingGameplayTag (Utilities|Casting|ToGameplayTagAssetInterface (Ability|GetAvatarActorfromActorInfo)) "(TagName=\"Event.CherrySmoke.Medallions.Leap.Cloaked\")")))
